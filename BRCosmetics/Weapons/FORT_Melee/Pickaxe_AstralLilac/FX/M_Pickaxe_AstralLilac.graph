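# Unlit translucent Niagara sprite material for the Astral Lilac pickaxe effects, drawing a procedural shape without textures.
# UVs get a power-based vertical squash, are centered and mirrored with absolute value, then a power curve and inverted sum form the mask.
# A static switch, on by default, thresholds the mask into a hard edge instead of a soft gradient.
# Particle alpha multiplies the mask, while particle color feeds a shared emissive adjustment function.
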Material /BRCosmetics/Weapons/FORT_Melee/Pickaxe_AstralLilac/FX/M_Pickaxe_AstralLilac
graph {
 properties {bUsedWithNiagaraSprites=true, blendMode=BLEND_Translucent, pixelDepthOffsetMode=PDOM_Legacy, shadingModel=MSM_Unlit, usedShadingModels=Unlit}
 nodes [
  n1 = ParticleColor()
  MaterialFunctionCall(Emissive=@n1.RGB) {materialFunction=&MF_FX_EmissiveAdjustments}
  n2 = TextureCoordinate()
  n3 = ComponentMask(None=@n2) {g=true}
  n4 = ScalarParameter() {defaultValue=2, parameterName=Squash, sliderMax=2}
  n5 = ConstantBiasScale(None=AppendVector(A=ComponentMask(None=@n2) {r=true}, B=Power(Base=@n3, Exp=@n4) {constExponent=0.6})) {bias=-0.5, scale=2}
  n6 = MaterialFunctionCall(Float2=Power(Base=Abs(None=@n5), Exp=ScalarParameter() {defaultValue=0.5, parameterName=Power, sliderMax=10})) {materialFunction=&BreakOutFloat2Components}
  n7 = OneMinus(None=Add(A=@n6.R, B=@n6.R))
  n8 = Step(Y=@n7) {constX=0.2}
  Multiply(A=@n1.A, B=StaticSwitchParameter(True=OneMinus(None=@n8), False=@n7) {defaultValue=true, parameterName=IsHardEdge})
 ]
}
