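# Niagara swing trail for the AstralLilac pickaxe, built from three ribbon emitters that spawn short-lived anim trails from skeletal mesh bones or sockets.
# A pale blue main trail is joined by two offset top trails using bone/socket index 1, one shifting toward red and one yellow-to-orange with tapering width.
# Each emitter stacks two color-scale passes: time-of-day tint and alpha scaling, then a sharper fade curve.
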NiagaraSystem /BRCosmetics/Weapons/FORT_Melee/Pickaxe_AstralLilac/FX/NS_Pickaxe_AstralLilac_Trail
dependencies {
 usedDataInterfaces [
  &NiagaraDataInterfaceCurve
  &NiagaraDataInterfaceColorCurve
  &NiagaraDataInterfaceVectorCurve
  &NiagaraDataInterfaceSkeletalMesh
 ]
 usedDynamicInputs []
 usedModules [
  &self:SystemSpawnScript.SystemScriptSource.SystemScriptGraph.NiagaraNodeAssignment_0.SetVariables_#1
  &SystemState
 ]
}
emitters {
 Main_Trail {
  emitterUpdateScript [
   &EmitterState()
   &FN_SpawnAnimTrail(Spawn Interval=0)
  ]
  particleSpawnScript [
   &FN_InitializeAnimTrailBoneIndecies()
   &FN_InitializeAnimTrail(Color=<rgba 0.5221 0.5926 0.7913 0.5>, Lifetime=0.08, Position Offset=<0 0 -5>, SkeletalMesh=$System.Skeletal Mesh)
  ]
  particleUpdateScript [
   &FN_UpdateAnimTrail(Locked Particle RibbonWidth=1, SkeletalMesh=$System.Skeletal Mesh) {module=FN_UpdateAnimTrail001}
   &ScaleColor(
    Scale Alpha=&FloatFromCurve(
     FloatCurve=DataInterface() {propertyValues {Curve {keys [
         {arriveTangent=-40.3311, interpMode=RCIM_Cubic, leaveTangent=-40.3311, tangentMode=RCTM_User, time=0, value=2.5}
         {arriveTangent=-1.3327, interpMode=RCIM_Cubic, leaveTangent=-1.3327, tangentMode=RCTM_User, time=0.1, value=1.25}
         {interpMode=RCIM_Cubic, tangentMode=RCTM_User, time=1, value=0}
        ]}}}
     Scale Curve=&FN_Select_Float_from_Time_Of_Day(Morning=0.5, Day=0.4, Evening=0.5, Night=0.45)
    )
    Scale RGB=&MakeVectorFromLinearColorRGB(Color=&FNTimeOfDayColor(
      Night Color=<rgba 0.2 0.275 0.5 1>
      Morning Color=<rgba 0.5 0.45 0.4542 1>
      Day Color=<rgba 0.7 0.7 0.7 1>
      Evening Color=<rgba 0.65 0.4333 0.39 1>
     ))
    ScaleA=true
    ScaleRGB=true
   )
   &ScaleColor(
    Scale Alpha=&FloatFromCurve(FloatCurve=DataInterface() {propertyValues {Curve {keys [
         {arriveTangent=-21.3684, interpMode=RCIM_Cubic, leaveTangent=-21.3684, tangentMode=RCTM_User, time=0.1, value=1}
         {arriveTangent=-6.7025, interpMode=RCIM_Cubic, leaveTangent=-6.7025, tangentMode=RCTM_User, time=0.1531, value=0.3765}
         <at 0.3 0 RCIM_Cubic>
        ]}}})
    ScaleA=true
    ScaleRGB=true
   ) {module=ScaleColor001}
  ]
  rendererClasses [&NiagaraRibbonRendererProperties]
  renderers [{rendererClass=&NiagaraRibbonRendererProperties}]
 }
 Main_Trail_Top {
  emitterUpdateScript [
   &EmitterState()
   &FN_SpawnAnimTrail(Spawn Interval=0)
  ]
  particleSpawnScript [
   &FN_InitializeAnimTrailBoneIndecies(First Bone or Socket Index=1)
   &FN_InitializeAnimTrail(
    AnimTrailWidth=5
    Color=<rgba 0.521 0.6376 0.8963 1>
    Lifetime=0.08
    Position Offset=&TransformVector(Vector=<0 0 -2>, Destination Space=::World)
    SkeletalMesh=$System.Skeletal Mesh
   )
  ]
  particleUpdateScript [
   &FN_UpdateAnimTrail(Locked Particle RibbonWidth=1, SkeletalMesh=$System.Skeletal Mesh) {module=FN_UpdateAnimTrail001}
   &ScaleColor(
    Scale Alpha=&FloatFromCurve(
     FloatCurve=DataInterface() {propertyValues {Curve {keys [
         {arriveTangent=-8.0864, interpMode=RCIM_Cubic, leaveTangent=-8.0864, tangentMode=RCTM_User, time=0, value=2.5}
         {arriveTangent=-3.5406, interpMode=RCIM_Cubic, leaveTangent=-3.5406, tangentMode=RCTM_User, time=0.2, value=1.375}
         {interpMode=RCIM_Cubic, tangentMode=RCTM_User, time=1, value=0}
        ]}}}
     Scale Curve=&FN_Select_Float_from_Time_Of_Day(Morning=0.5, Day=0.4, Evening=0.5, Night=0.45)
    )
    Scale RGB=&MakeVectorFromLinearColorRGB(Color=&FNTimeOfDayColor(
      Night Color=<rgba 0.2 0.275 0.5 1>
      Morning Color=<rgba 0.5 0.45 0.4542 1>
      Day Color=<rgba 0.7 0.7 0.7 1>
      Evening Color=<rgba 0.65 0.4333 0.39 1>
     ))
    ScaleA=true
    ScaleRGB=true
   )
   &ScaleColor(
    Scale Alpha=&FloatFromCurve(FloatCurve=DataInterface() {propertyValues {Curve {keys [
         {arriveTangent=-21.3684, interpMode=RCIM_Cubic, leaveTangent=-21.3684, tangentMode=RCTM_User, time=0.1, value=1}
         {arriveTangent=-0.6511, interpMode=RCIM_Cubic, leaveTangent=-0.6511, tangentMode=RCTM_User, time=0.1495, value=0.1}
         <at 0.3 0 RCIM_Cubic>
        ]}}})
    Scale RGB=&VectorFromCurve(VectorCurve=DataInterface() {propertyValues {
       XCurve {keys [<at 0 1 RCIM_Cubic>,<at 1 1>]}
       YCurve {keys [{arriveTangent=-0.929, interpMode=RCIM_Cubic, leaveTangent=-0.929, tangentMode=RCTM_User, time=0, value=1}, <at 1 0.5>]}
       ZCurve {keys [{arriveTangent=-0.929, interpMode=RCIM_Cubic, leaveTangent=-0.929, tangentMode=RCTM_User, time=0, value=1}, <at 1 0.5>]}
      }})
    ScaleA=true
    ScaleRGB=true
   ) {module=ScaleColor001}
  ]
  rendererClasses [&NiagaraRibbonRendererProperties]
  renderers [{rendererClass=&NiagaraRibbonRendererProperties}]
 }
 Main_Trail_Top_Yellow {
  emitterUpdateScript [
   &EmitterState()
   &FN_SpawnAnimTrail(Spawn Interval=0)
  ]
  particleSpawnScript [
   &FN_InitializeAnimTrailBoneIndecies(First Bone or Socket Index=1)
   &FN_InitializeAnimTrail(
    AnimTrailWidth=4
    Lifetime=0.1
    Position Offset=&TransformVector(Vector=<0 0 -7>, Destination Space=::World)
    SkeletalMesh=$System.Skeletal Mesh
   )
  ]
  particleUpdateScript [
   &FN_UpdateAnimTrail(Locked Particle RibbonWidth=1, SkeletalMesh=$System.Skeletal Mesh) {module=FN_UpdateAnimTrail001}
   &ScaleColor(
    Scale Alpha=&FloatFromCurve(
     FloatCurve=DataInterface() {propertyValues {Curve {keys [
         {arriveTangent=-8.0864, interpMode=RCIM_Cubic, leaveTangent=-8.0864, tangentMode=RCTM_User, time=0, value=2.5}
         {arriveTangent=-3.5406, interpMode=RCIM_Cubic, leaveTangent=-3.5406, tangentMode=RCTM_User, time=0.2, value=1.375}
         {interpMode=RCIM_Cubic, tangentMode=RCTM_User, time=1, value=0}
        ]}}}
     Scale Curve=&FN_Select_Float_from_Time_Of_Day(Morning=0.5, Day=0.4, Evening=0.5, Night=0.45)
    )
    Scale RGB=&MakeVectorFromLinearColorRGB(Color=&FNTimeOfDayColor(
      Night Color=<rgba 0.2 0.275 0.5 1>
      Morning Color=<rgba 0.5 0.45 0.4542 1>
      Day Color=<rgba 0.7 0.7 0.7 1>
      Evening Color=<rgba 0.65 0.4333 0.39 1>
     ))
    ScaleA=true
    ScaleRGB=true
   )
   &ScaleColor(
    Scale Alpha=&FloatFromCurve(FloatCurve=DataInterface() {propertyValues {Curve {keys [
         {arriveTangent=-36.7476, interpMode=RCIM_Cubic, leaveTangent=-36.7476, tangentMode=RCTM_User, time=0.1, value=1}
         {arriveTangent=-0.6511, interpMode=RCIM_Cubic, leaveTangent=-0.6511, tangentMode=RCTM_User, time=0.1495, value=0.1}
         <at 0.3 0 RCIM_Cubic>
        ]}}})
    Scale RGB=&MakeVectorFromLinearColorRGB(Color=&ColorFromCurve(ColorCurve=DataInterface() {propertyValues {
        AlphaCurve {keys []}
        BlueCurve {keys [<at 0.0155 0.0071 RCIM_Cubic>,<at 0.143 0.0241>,<at 0.7859 0.0063>]}
        GreenCurve {keys [
          {arriveTangent=-34.7445, interpMode=RCIM_Cubic, leaveTangent=-34.7445, tangentMode=RCTM_User, time=0.0155, value=1.7717}
          <at 0.143 0.2158>
          <at 0.7859 0.0564>
         ]}
        RedCurve {keys [
          {arriveTangent=-34.7445, interpMode=RCIM_Cubic, leaveTangent=-34.7445, tangentMode=RCTM_User, time=0.0155, value=2}
          <at 0.143 0.5781>
          <at 0.7859 0.151>
         ]}
       }}))
    ScaleA=true
    ScaleRGB=true
   ) {module=ScaleColor001}
   &ScaleRibbonWidth(Ribbon Width Scale=&FloatFromCurve(FloatCurve=DataInterface() {propertyValues {Curve {keys [<at 0 1>,<at 1 0>]}}}))
  ]
  rendererClasses [&NiagaraRibbonRendererProperties]
  renderers [{rendererClass=&NiagaraRibbonRendererProperties}]
 }
}
summary {
 systemName=NS_Pickaxe_AstralLilac_Trail
 emitters [
  {emitterName=Main_Trail, rendererClasses [&NiagaraRibbonRendererProperties]}
  {emitterName=Main_Trail_Top, rendererClasses [&NiagaraRibbonRendererProperties]}
  {emitterName=Main_Trail_Top_Yellow, rendererClasses [&NiagaraRibbonRendererProperties]}
 ]
 userVariables [User.RibbonIDOffset {type=NiagaraInt32}]
}
